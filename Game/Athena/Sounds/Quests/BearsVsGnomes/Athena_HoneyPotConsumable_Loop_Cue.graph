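# Looping bee ambience cue for the Athena honey pot consumable, filed under the BearsVsGnomes quest sounds.
# An Enveloper wraps a looping bee wave player, fading volume in over about half a second.
# It plays at reduced volume in the Athena ambient sound class.
# Inline attenuation and concurrency overrides set the falloff and cap simultaneous instances, taking precedence over the referenced consumable ambience attenuation asset.
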SoundCue /Game/Athena/Sounds/Quests/BearsVsGnomes/Athena_HoneyPotConsumable_Loop_Cue
root=Enveloper_0
nodes [
 {
  id=Enveloper_0
  children [WavePlayer_1]
  properties {pitchCurve {editorCurveData {keys [{time=0}]}}, volumeCurve {editorCurveData {keys [<at -0.0027 -0.0016>,<at 0 0.0028>,<at 0.5 1>]}}}
 }
 {id=WavePlayer_1, properties {bLooping=true, soundWaveAssetPtr=&/SFX_HoneyPotConsumable_Bee_Loop_01}}
]
properties {
 attenuationSettings=/Game/Sounds/Attenuation/Fort_Consumable_Amb_Att
 bOverrideAttenuation=true
 bOverrideConcurrency=true
 maxDistance=2900
 soundClassObject=&Fort_Ambient_Athena
 volumeMultiplier=0.501
 attenuationOverrides {falloffDistance=2500}
 concurrencyOverrides {maxCount=3}
}
