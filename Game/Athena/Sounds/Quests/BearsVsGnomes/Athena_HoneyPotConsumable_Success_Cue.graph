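# Sound cue for a successful honey pot consumable in the Bears vs Gnomes quest, playing a teddy bear roar.
# A single wave player is the root, routed to a non-critical in-game UI sound class with boosted volume.
# Cue-level attenuation and concurrency overrides give it a distance falloff and allow only one instance at a time, stopping the oldest when retriggered.
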SoundCue /Game/Athena/Sounds/Quests/BearsVsGnomes/Athena_HoneyPotConsumable_Success_Cue
root=WavePlayer_0
nodes [{id=WavePlayer_0, properties {soundWaveAssetPtr=&/SFX_HoneyPotConsumable_Success_TeddyBearRoar}}]
properties {
 bOverrideAttenuation=true
 bOverrideConcurrency=true
 maxDistance=2400
 soundClassObject=&Fort_UI_InGame_NonCritical
 volumeMultiplier=1.5
 attenuationOverrides {falloffDistance=2000}
 concurrencyOverrides {maxCount=1, resolutionRule=StopOldest}
}
