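# Sound cue for the gnome giggle in the Bears vs Gnomes quest audio.
# A single wave player plays one gnome laugh recording, with no randomization or modulation.
# It uses the non-critical in-game UI sound class and the player voice attenuation asset, though an attenuation override with its own falloff distance is enabled on the cue.
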SoundCue /Game/Athena/Sounds/Quests/BearsVsGnomes/BvG_Gnomes_Giggle
root=WavePlayer_1
nodes [{id=WavePlayer_1, properties {soundWaveAssetPtr=&/Quest_Gnome_Bears_Gnome_Laugh_01}}]
properties {
 attenuationSettings=/Game/Sounds/Attenuation/Fort_VO_Player_Att
 bOverrideAttenuation=true
 maxDistance=2400
 soundClassObject=&Fort_UI_InGame_NonCritical
 attenuationOverrides {falloffDistance=2000}
}
